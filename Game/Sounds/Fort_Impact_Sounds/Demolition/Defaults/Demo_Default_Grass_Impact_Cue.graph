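# Default demolition impact sound cue for grass, built from medium hay impact recordings.
# A quality level root selects one of four branches: the first randomizes three waves, the middle two randomize two, and the last plays a single wave.
# Each branch applies slight random pitch modulation.
# The cue uses the impacts sound class, small melee weapon attenuation, and a four-max object impact concurrency set, with a prevent-new rule in its overrides.
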SoundCue /Game/Sounds/Fort_Impact_Sounds/Demolition/Defaults/Demo_Default_Grass_Impact_Cue
root=QualityLevel_0
nodes [
 {id=Modulator_0, children [Random_0], properties {pitchMax=1.07, pitchMin=0.92}}
 {id=Modulator_1, children [Random_1], properties {pitchMax=1.07, pitchMin=0.92}}
 {id=Modulator_2, children [Random_22], properties {pitchMax=1.07, pitchMin=0.92}}
 {id=Modulator_3, children [WavePlayer_4], properties {pitchMax=1.1, pitchMin=0.9}}
 {id=QualityLevel_0, children [Modulator_2,Modulator_0,Modulator_1,Modulator_3]}
 {id=Random_0, children [WavePlayer_0,WavePlayer_1], properties {weights [1,1]}}
 {id=Random_1, children [WavePlayer_2,WavePlayer_3], properties {weights [1,1]}}
 {id=Random_22, children [WavePlayer_105, WavePlayer_106, WavePlayer_108], properties {weights [1,1,1]}}
 {id=WavePlayer_0, properties {soundWaveAssetPtr=/Game/Sounds/Fort_Impact_Sounds/Demolition/Hay/Hay_Impact_Med4}}
 {id=WavePlayer_1, properties {soundWaveAssetPtr=/Game/Sounds/Fort_Impact_Sounds/Demolition/Hay/Hay_Impact_Med1}}
 {id=WavePlayer_105, properties {soundWaveAssetPtr=&Hay_Impact_Med4}}
 {id=WavePlayer_106, properties {soundWaveAssetPtr=&Hay_Impact_Med1}}
 {id=WavePlayer_108, properties {soundWaveAssetPtr=/Game/Sounds/Fort_Impact_Sounds/Demolition/Hay/Hay_Impact_Med3}}
 {id=WavePlayer_2, properties {soundWaveAssetPtr=&Hay_Impact_Med4}}
 {id=WavePlayer_3, properties {soundWaveAssetPtr=&Hay_Impact_Med1}}
 {id=WavePlayer_4, properties {soundWaveAssetPtr=&Hay_Impact_Med4}}
]
properties {
 attenuationSettings=&Fort_Weapon_Melee_Sml_Att
 maxDistance=2800
 soundClassObject=&Fort_Impacts
 volumeMultiplier=1.696
 concurrencyOverrides {resolutionRule=PreventNew}
 concurrencySet [&CONC_Object_Impacts_FourMax]
}
