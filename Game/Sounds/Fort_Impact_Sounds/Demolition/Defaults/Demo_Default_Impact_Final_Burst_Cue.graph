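# Sound cue named as the default final-burst impact for demolition, holding only playback settings: its root is null and it has no sound nodes.
# It routes to the Fort_Impacts_PSM sound class at reduced volume, with medium-sized-object destruction attenuation.
# It references a shared final-burst concurrency set and also carries concurrency override values for a small voice cap with distance-based volume scaling.
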SoundCue /Game/Sounds/Fort_Impact_Sounds/Demolition/Defaults/Demo_Default_Impact_Final_Burst_Cue
root=null
nodes []
properties {
 attenuationSettings=/Game/Sounds/Attenuation/Fort_Demolition_Destroy_Medium_Sized_Object_Att
 maxDistance=6400
 soundClassObject=&Fort_Impacts_PSM
 volumeMultiplier=0.701
 concurrencyOverrides {maxCount=4, volumeScale=0.5, volumeScaleMode=Distance}
 concurrencySet [/Game/Sounds/Fort_Impact_Sounds/Demolition/Settings/CONC_Demo_Impact_Final_Burst]
}
