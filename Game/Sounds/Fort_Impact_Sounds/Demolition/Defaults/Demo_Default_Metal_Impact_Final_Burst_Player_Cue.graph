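# Sound cue for the player's default metal demolition final-burst impact, routed to an impacts sound class with a final-burst concurrency set.
# A mixer layers close, medium and far branches, with the medium and far layers at half input volume.
# Each branch has its own crossfade attenuation, a pitch-varying modulator with fixed volume, and a random pick from its distance-specific wave variants.
# The random nodes are excluded from branch culling.
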SoundCue /Game/Sounds/Fort_Impact_Sounds/Demolition/Defaults/Demo_Default_Metal_Impact_Final_Burst_Player_Cue
root=Mixer_2
nodes [
 {
  id=Attenuation_0
  children [Modulator_10]
  properties {attenuationSettings=/Game/Sounds/Attenuation/Fort_Demolition_Crossfade_Global_Hit_Far_Att}
 }
 {
  id=Attenuation_1
  children [Modulator_5]
  properties {attenuationSettings=/Game/Sounds/Attenuation/Fort_Demolition_Crossfade_Global_Hit_Medium_Att}
 }
 {
  id=Attenuation_2
  children [Modulator_1]
  properties {attenuationSettings=/Game/Sounds/Attenuation/Fort_Demolition_Crossfade_Global_Hit_Close_Att}
 }
 {id=Mixer_2, children [Attenuation_2, Attenuation_1, Attenuation_0], properties {inputVolume [1,0.5,0.5]}}
 {id=Modulator_1, children [Random_2], properties {volumeMax=1, volumeMin=1}}
 {id=Modulator_10, children [Random_5], properties {volumeMax=1, volumeMin=1}}
 {id=Modulator_5, children [Random_3], properties {volumeMax=1, volumeMin=1}}
 {
  id=Random_2
  children [
   WavePlayer_3
   WavePlayer_20
   WavePlayer_21
   WavePlayer_22
   WavePlayer_23
   WavePlayer_24
   WavePlayer_25
   WavePlayer_26
   WavePlayer_27
   WavePlayer_28
   WavePlayer_29
   WavePlayer_30
  ]
  properties {bShouldExcludeFromBranchCulling=true, weights [1,1,1,1,1,1,1,1,1,1,1,1]}
 }
 {
  id=Random_3
  children [WavePlayer_32, WavePlayer_39, WavePlayer_40, WavePlayer_41, WavePlayer_42, WavePlayer_49]
  properties {bShouldExcludeFromBranchCulling=true, weights [1,1,1,1,1,1]}
 }
 {
  id=Random_5
  children [WavePlayer_50, WavePlayer_51, WavePlayer_52, WavePlayer_53, WavePlayer_54]
  properties {bShouldExcludeFromBranchCulling=true, weights [1,1,1,1,1]}
 }
 {
  id=WavePlayer_20
  properties {soundWaveAssetPtr=&/Demo_Default_Metal_Impact_Final_Burst_Close_02}
 }
 {
  id=WavePlayer_21
  properties {soundWaveAssetPtr=&/Demo_Default_Metal_Impact_Final_Burst_Close_03}
 }
 {
  id=WavePlayer_22
  properties {soundWaveAssetPtr=&/Demo_Default_Metal_Impact_Final_Burst_Close_04}
 }
 {
  id=WavePlayer_23
  properties {soundWaveAssetPtr=&/Demo_Default_Metal_Impact_Final_Burst_Close_05}
 }
 {
  id=WavePlayer_24
  properties {soundWaveAssetPtr=&/Demo_Default_Metal_Impact_Final_Burst_Close_06}
 }
 {
  id=WavePlayer_25
  properties {soundWaveAssetPtr=&/Demo_Default_Metal_Impact_Final_Burst_Close_07}
 }
 {
  id=WavePlayer_26
  properties {soundWaveAssetPtr=&/Demo_Default_Metal_Impact_Final_Burst_Close_08}
 }
 {
  id=WavePlayer_27
  properties {soundWaveAssetPtr=&/Demo_Default_Metal_Impact_Final_Burst_Close_09}
 }
 {
  id=WavePlayer_28
  properties {soundWaveAssetPtr=&/Demo_Default_Metal_Impact_Final_Burst_Close_10}
 }
 {
  id=WavePlayer_29
  properties {soundWaveAssetPtr=&/Demo_Default_Metal_Impact_Final_Burst_Close_11}
 }
 {
  id=WavePlayer_3
  properties {soundWaveAssetPtr=&/Demo_Default_Metal_Impact_Final_Burst_Close_01}
 }
 {
  id=WavePlayer_30
  properties {soundWaveAssetPtr=&/Demo_Default_Metal_Impact_Final_Burst_Close_12}
 }
 {
  id=WavePlayer_32
  properties {soundWaveAssetPtr=&/Demo_Default_Metal_Impact_Final_Burst_Mid_01}
 }
 {
  id=WavePlayer_39
  properties {soundWaveAssetPtr=&/Demo_Default_Metal_Impact_Final_Burst_Mid_02}
 }
 {
  id=WavePlayer_40
  properties {soundWaveAssetPtr=&/Demo_Default_Metal_Impact_Final_Burst_Mid_03}
 }
 {
  id=WavePlayer_41
  properties {soundWaveAssetPtr=&/Demo_Default_Metal_Impact_Final_Burst_Mid_04}
 }
 {
  id=WavePlayer_42
  properties {soundWaveAssetPtr=&/Demo_Default_Metal_Impact_Final_Burst_Mid_05}
 }
 {
  id=WavePlayer_49
  properties {soundWaveAssetPtr=&/Demo_Default_Metal_Impact_Final_Burst_Mid_06}
 }
 {
  id=WavePlayer_50
  properties {soundWaveAssetPtr=&/Demo_Default_Metal_Impact_Final_Burst_Dist_01}
 }
 {
  id=WavePlayer_51
  properties {soundWaveAssetPtr=&/Demo_Default_Metal_Impact_Final_Burst_Dist_02}
 }
 {
  id=WavePlayer_52
  properties {soundWaveAssetPtr=&/Demo_Default_Metal_Impact_Final_Burst_Dist_03}
 }
 {
  id=WavePlayer_53
  properties {soundWaveAssetPtr=&/Demo_Default_Metal_Impact_Final_Burst_Dist_04}
 }
 {
  id=WavePlayer_54
  properties {soundWaveAssetPtr=&/Demo_Default_Metal_Impact_Final_Burst_Dist_05}
 }
]
properties {
 maxDistance=8000
 soundClassObject=&Fort_Impacts_PSM
 volumeMultiplier=1.35
 concurrencyOverrides {maxCount=1}
 concurrencySet [/Game/Sounds/Fort_Impact_Sounds/Demolition/Settings/CONC_Demo_Impact_Final_Burst]
}
