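# Default rock impact cue for explosive hits in demolition, routed to the impacts sound class with a shared object-impact concurrency set.
# A quality-level root selects one of four mixers, each with fewer random wave variations than the last, down to single waves.
# Each mixer layers a full-volume rock impact with a quieter, slightly delayed stereo impact, each through its own modulator and attenuation override.
# The cue raises overall pitch and references small melee weapon attenuation settings.
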SoundCue /Game/Sounds/Fort_Impact_Sounds/Demolition/Defaults/Demo_Default_Rock_Impact_Explosive_Cue
root=QualityLevel_0
nodes [
 {
  id=Attenuation_0
  children [Delay_1]
  properties {bOverrideAttenuation=true, attenuationOverrides {falloffDistance=900, attenuationShapeExtents {x=100}}}
 }
 {id=Attenuation_1, children [Modulator_3], properties {bOverrideAttenuation=true}}
 {
  id=Attenuation_2
  children [Delay_2]
  properties {bOverrideAttenuation=true, attenuationOverrides {falloffDistance=900, attenuationShapeExtents {x=100}}}
 }
 {id=Attenuation_3, children [Modulator_5], properties {bOverrideAttenuation=true}}
 {
  id=Attenuation_4
  children [Delay_0]
  properties {bOverrideAttenuation=true, attenuationOverrides {falloffDistance=900, attenuationShapeExtents {x=100}}}
 }
 {id=Attenuation_5, children [Modulator_1], properties {bOverrideAttenuation=true}}
 {
  id=Attenuation_6
  children [Delay_3]
  properties {bOverrideAttenuation=true, attenuationOverrides {falloffDistance=900, attenuationShapeExtents {x=100}}}
 }
 {id=Attenuation_7, children [Modulator_7], properties {bOverrideAttenuation=true}}
 {id=Delay_0, children [Modulator_0], properties {delayMax=0.1, delayMin=0.05}}
 {id=Delay_1, children [Modulator_2], properties {delayMax=0.1, delayMin=0.05}}
 {id=Delay_2, children [Modulator_4], properties {delayMax=0.1, delayMin=0.05}}
 {id=Delay_3, children [Modulator_6], properties {delayMax=0.1, delayMin=0.05}}
 {id=Mixer_0, children [Attenuation_1, Attenuation_0], properties {inputVolume [1,0.2]}}
 {id=Mixer_1, children [Attenuation_3, Attenuation_2], properties {inputVolume [1,0.2]}}
 {id=Mixer_2, children [Attenuation_7, Attenuation_6], properties {inputVolume [1,0.2]}}
 {id=Mixer_3, children [Attenuation_5, Attenuation_4], properties {inputVolume [1,0.2]}}
 {id=Modulator_0, children [Random_15]}
 {id=Modulator_1, children [Random_14]}
 {id=Modulator_2, children [Random_1]}
 {id=Modulator_3, children [Random_0]}
 {id=Modulator_4, children [WavePlayer_11]}
 {id=Modulator_5, children [Random_2]}
 {id=Modulator_6, children [WavePlayer_15]}
 {id=Modulator_7, children [WavePlayer_14]}
 {id=QualityLevel_0, children [Mixer_3,Mixer_0,Mixer_1,Mixer_2]}
 {id=Random_0, children [WavePlayer_1,WavePlayer_0,WavePlayer_3], properties {weights [1,1,1]}}
 {id=Random_1, children [WavePlayer_5,WavePlayer_6], properties {weights [1,1]}}
 {id=Random_14, children [WavePlayer_109, WavePlayer_108, WavePlayer_114, WavePlayer_112], properties {weights [1,1,1,1]}}
 {id=Random_15, children [WavePlayer_119, WavePlayer_122, WavePlayer_123, WavePlayer_117], properties {weights [1,1,1,1]}}
 {id=Random_2, children [WavePlayer_9,WavePlayer_8], properties {weights [1,1]}}
 {id=WavePlayer_0, properties {soundWaveAssetPtr=&/rock_impact_05}}
 {id=WavePlayer_1, properties {soundWaveAssetPtr=&/rock_impact_04}}
 {id=WavePlayer_108, properties {soundWaveAssetPtr=&rock_impact_05}}
 {id=WavePlayer_109, properties {soundWaveAssetPtr=&rock_impact_04}}
 {id=WavePlayer_11, properties {soundWaveAssetPtr=&/rock_impact_stereo_03}}
 {id=WavePlayer_112, properties {soundWaveAssetPtr=&/rock_impact_08}}
 {id=WavePlayer_114, properties {soundWaveAssetPtr=&/rock_impact_07}}
 {id=WavePlayer_117, properties {soundWaveAssetPtr=&/rock_impact_stereo_01}}
 {id=WavePlayer_119, properties {soundWaveAssetPtr=&rock_impact_stereo_03}}
 {id=WavePlayer_122, properties {soundWaveAssetPtr=&/rock_impact_stereo_06}}
 {id=WavePlayer_123, properties {soundWaveAssetPtr=&/rock_impact_stereo_07}}
 {id=WavePlayer_14, properties {soundWaveAssetPtr=&rock_impact_04}}
 {id=WavePlayer_15, properties {soundWaveAssetPtr=&rock_impact_stereo_03}}
 {id=WavePlayer_3, properties {soundWaveAssetPtr=&rock_impact_07}}
 {id=WavePlayer_5, properties {soundWaveAssetPtr=&rock_impact_stereo_03}}
 {id=WavePlayer_6, properties {soundWaveAssetPtr=&rock_impact_stereo_06}}
 {id=WavePlayer_8, properties {soundWaveAssetPtr=&rock_impact_05}}
 {id=WavePlayer_9, properties {soundWaveAssetPtr=&rock_impact_04}}
]
properties {
 attenuationSettings=&Fort_Weapon_Melee_Sml_Att
 maxDistance=4000
 pitchMultiplier=1.5
 soundClassObject=&Fort_Impacts
 volumeMultiplier=1.01
 concurrencyOverrides {resolutionRule=PreventNew}
 concurrencySet [&CONC_Object_Impacts_FourMax]
}
