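# Sound cue for the enemy variant of the default rock demolition final-burst impact.
# A mixer layers close, mid and distant branches, each using its own crossfade attenuation, with mid and far mixed lower and overriding falloff.
# Each branch randomly picks a wave and passes it through a modulator that keeps volume fixed, leaving default pitch variation.
# The cue routes to the impacts sound class and uses the final-burst concurrency set.
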SoundCue /Game/Sounds/Fort_Impact_Sounds/Demolition/Defaults/Demo_Default_Rock_Impact_Final_Burst_Enemy_Cue
root=Mixer_1
nodes [
 {
  id=Attenuation_1
  children [Modulator_2]
  properties {attenuationSettings=/Game/Sounds/Attenuation/Fort_Demolition_Crossfade_Global_Hit_Close_Att}
 }
 {
  id=Attenuation_6
  children [Modulator_1]
  properties {
   attenuationSettings=/Game/Sounds/Attenuation/Fort_Demolition_Crossfade_Global_Hit_Medium_Att
   attenuationOverrides {falloffDistance=900, attenuationShapeExtents {x=100}}
  }
 }
 {
  id=Attenuation_7
  children [Modulator_0]
  properties {
   attenuationSettings=/Game/Sounds/Attenuation/Fort_Demolition_Crossfade_Global_Hit_Far_Att
   attenuationOverrides {falloffDistance=900, attenuationShapeExtents {x=100}}
  }
 }
 {id=Mixer_1, children [Attenuation_1, Attenuation_6, Attenuation_7], properties {inputVolume [1,0.75,0.75]}}
 {id=Modulator_0, children [Random_7], properties {volumeMax=1, volumeMin=1}}
 {id=Modulator_1, children [Random_3], properties {volumeMax=1, volumeMin=1}}
 {id=Modulator_2, children [Random_0], properties {volumeMax=1, volumeMin=1}}
 {
  id=Random_0
  children [
   WavePlayer_18
   WavePlayer_19
   WavePlayer_20
   WavePlayer_21
   WavePlayer_22
   WavePlayer_23
   WavePlayer_24
   WavePlayer_25
   WavePlayer_26
   WavePlayer_27
   WavePlayer_28
   WavePlayer_29
  ]
  properties {bShouldExcludeFromBranchCulling=true, weights [1,1,1,1,1,1,1,1,1,1,1,1]}
 }
 {
  id=Random_3
  children [WavePlayer_36, WavePlayer_37, WavePlayer_38, WavePlayer_39, WavePlayer_40]
  properties {bShouldExcludeFromBranchCulling=true, weights [1,1,1,1,1]}
 }
 {
  id=Random_7
  children [WavePlayer_41, WavePlayer_42, WavePlayer_43, WavePlayer_44, WavePlayer_45]
  properties {bShouldExcludeFromBranchCulling=true, weights [1,1,1,1,1]}
 }
 {
  id=WavePlayer_18
  properties {soundWaveAssetPtr=&/Demo_Default_Rock_Impact_Final_Burst_Close_01}
 }
 {
  id=WavePlayer_19
  properties {soundWaveAssetPtr=&/Demo_Default_Rock_Impact_Final_Burst_Close_02}
 }
 {
  id=WavePlayer_20
  properties {soundWaveAssetPtr=&/Demo_Default_Rock_Impact_Final_Burst_Close_03}
 }
 {
  id=WavePlayer_21
  properties {soundWaveAssetPtr=&/Demo_Default_Rock_Impact_Final_Burst_Close_04}
 }
 {
  id=WavePlayer_22
  properties {soundWaveAssetPtr=&/Demo_Default_Rock_Impact_Final_Burst_Close_05}
 }
 {
  id=WavePlayer_23
  properties {soundWaveAssetPtr=&/Demo_Default_Rock_Impact_Final_Burst_Close_06}
 }
 {
  id=WavePlayer_24
  properties {soundWaveAssetPtr=&/Demo_Default_Rock_Impact_Final_Burst_Close_07}
 }
 {
  id=WavePlayer_25
  properties {soundWaveAssetPtr=&/Demo_Default_Rock_Impact_Final_Burst_Close_08}
 }
 {
  id=WavePlayer_26
  properties {soundWaveAssetPtr=&/Demo_Default_Rock_Impact_Final_Burst_Close_09}
 }
 {
  id=WavePlayer_27
  properties {soundWaveAssetPtr=&/Demo_Default_Rock_Impact_Final_Burst_Close_10}
 }
 {
  id=WavePlayer_28
  properties {soundWaveAssetPtr=&/Demo_Default_Rock_Impact_Final_Burst_Close_11}
 }
 {
  id=WavePlayer_29
  properties {soundWaveAssetPtr=&/Demo_Default_Rock_Impact_Final_Burst_Close_12}
 }
 {
  id=WavePlayer_36
  properties {soundWaveAssetPtr=&/Demo_Default_Rock_Impact_Final_Burst_Mid_01}
 }
 {
  id=WavePlayer_37
  properties {soundWaveAssetPtr=&/Demo_Default_Rock_Impact_Final_Burst_Mid_02}
 }
 {
  id=WavePlayer_38
  properties {soundWaveAssetPtr=&/Demo_Default_Rock_Impact_Final_Burst_Mid_03}
 }
 {
  id=WavePlayer_39
  properties {soundWaveAssetPtr=&/Demo_Default_Rock_Impact_Final_Burst_Mid_04}
 }
 {
  id=WavePlayer_40
  properties {soundWaveAssetPtr=&/Demo_Default_Rock_Impact_Final_Burst_Mid_05}
 }
 {
  id=WavePlayer_41
  properties {soundWaveAssetPtr=&/Demo_Default_Rock_Impact_Final_Burst_Dist_01}
 }
 {
  id=WavePlayer_42
  properties {soundWaveAssetPtr=&/Demo_Default_Rock_Impact_Final_Burst_Dist_02}
 }
 {
  id=WavePlayer_43
  properties {soundWaveAssetPtr=&/Demo_Default_Rock_Impact_Final_Burst_Dist_03}
 }
 {
  id=WavePlayer_44
  properties {soundWaveAssetPtr=&/Demo_Default_Rock_Impact_Final_Burst_Dist_04}
 }
 {
  id=WavePlayer_45
  properties {soundWaveAssetPtr=&/Demo_Default_Rock_Impact_Final_Burst_Dist_05}
 }
]
properties {
 maxDistance=8000
 soundClassObject=&Fort_Impacts_PSM
 volumeMultiplier=1.589
 concurrencyOverrides {maxCount=1}
 concurrencySet [/Game/Sounds/Fort_Impact_Sounds/Demolition/Settings/CONC_Demo_Impact_Final_Burst]
}
