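# Default demolition sound cue for rock final-debris impacts.
# A Random node picks one of six equally weighted debris waves and feeds a Modulator that varies pitch while keeping volume fixed.
# The cue uses the demolition debris attenuation, the Fort impacts sound class and a dedicated final-debris concurrency set, with a boosted volume multiplier.
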SoundCue /Game/Sounds/Fort_Impact_Sounds/Demolition/Defaults/Demo_Default_Rock_Impact_Final_Debris_Cue
root=Modulator_1
nodes [
 {id=Modulator_1, children [Random_2], properties {pitchMax=1.2, volumeMax=1, volumeMin=1}}
 {
  id=Random_2
  children [WavePlayer_12, WavePlayer_13, WavePlayer_14, WavePlayer_15, WavePlayer_16, WavePlayer_17]
  properties {bShouldExcludeFromBranchCulling=true, weights [1,1,1,1,1,1]}
 }
 {
  id=WavePlayer_12
  properties {soundWaveAssetPtr=&/Demo_Default_Rock_Impact_Final_Debris_01}
 }
 {
  id=WavePlayer_13
  properties {soundWaveAssetPtr=&/Demo_Default_Rock_Impact_Final_Debris_02}
 }
 {
  id=WavePlayer_14
  properties {soundWaveAssetPtr=&/Demo_Default_Rock_Impact_Final_Debris_03}
 }
 {
  id=WavePlayer_15
  properties {soundWaveAssetPtr=&/Demo_Default_Rock_Impact_Final_Debris_04}
 }
 {
  id=WavePlayer_16
  properties {soundWaveAssetPtr=&/Demo_Default_Rock_Impact_Final_Debris_05}
 }
 {
  id=WavePlayer_17
  properties {soundWaveAssetPtr=&/Demo_Default_Rock_Impact_Final_Debris_06}
 }
]
properties {
 attenuationSettings=&/Demo_Debris_Att
 maxDistance=8000
 soundClassObject=&Fort_Impacts_PSM
 volumeMultiplier=1.884
 concurrencyOverrides {maxCount=1}
 concurrencySet [/Game/Sounds/Fort_Impact_Sounds/Demolition/Settings/CONC_Demo_Impact_Final_Debris]
}
