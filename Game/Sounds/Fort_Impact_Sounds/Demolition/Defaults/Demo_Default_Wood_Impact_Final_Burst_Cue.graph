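# Sound cue for the default final-burst impact of wood demolition, layered by listener distance.
# A mixer blends close, medium and far branches, the latter two at half input volume, each with its own crossfade attenuation.
# Each branch randomly picks an equally weighted wave from its distance-specific set, excluded from branch culling, and passes it through a modulator with fixed volume.
# The cue uses an impacts sound class and a shared final-burst concurrency asset.
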SoundCue /Game/Sounds/Fort_Impact_Sounds/Demolition/Defaults/Demo_Default_Wood_Impact_Final_Burst_Cue
root=Mixer_2
nodes [
 {
  id=Attenuation_4
  children [Modulator_7]
  properties {attenuationSettings=/Game/Sounds/Attenuation/Fort_Demolition_Crossfade_Global_Hit_Far_Att}
 }
 {
  id=Attenuation_7
  children [Modulator_2]
  properties {attenuationSettings=/Game/Sounds/Attenuation/Fort_Demolition_Crossfade_Global_Hit_Medium_Att}
 }
 {
  id=Attenuation_8
  children [Modulator_0]
  properties {attenuationSettings=/Game/Sounds/Attenuation/Fort_Demolition_Crossfade_Global_Hit_Close_Att}
 }
 {id=Mixer_2, children [Attenuation_8, Attenuation_7, Attenuation_4], properties {inputVolume [1,0.5,0.5]}}
 {id=Modulator_0, children [Random_3], properties {volumeMax=1, volumeMin=1}}
 {id=Modulator_2, children [Random_4], properties {volumeMax=1, volumeMin=1}}
 {id=Modulator_7, children [Random_5], properties {volumeMax=1, volumeMin=1}}
 {
  id=Random_3
  children [
   WavePlayer_12
   WavePlayer_13
   WavePlayer_14
   WavePlayer_15
   WavePlayer_16
   WavePlayer_17
   WavePlayer_18
   WavePlayer_19
   WavePlayer_20
   WavePlayer_21
   WavePlayer_22
   WavePlayer_23
  ]
  properties {bShouldExcludeFromBranchCulling=true, weights [1,1,1,1,1,1,1,1,1,1,1,1]}
 }
 {
  id=Random_4
  children [WavePlayer_24, WavePlayer_25, WavePlayer_26, WavePlayer_27, WavePlayer_28]
  properties {bShouldExcludeFromBranchCulling=true, weights [1,1,1,1,1]}
 }
 {
  id=Random_5
  children [WavePlayer_29, WavePlayer_30, WavePlayer_46, WavePlayer_47, WavePlayer_48, WavePlayer_49]
  properties {bShouldExcludeFromBranchCulling=true, weights [1,1,1,1,1,1]}
 }
 {
  id=WavePlayer_12
  properties {soundWaveAssetPtr=&/Demo_Default_Wood_Impact_Final_Burst_Close_01}
 }
 {
  id=WavePlayer_13
  properties {soundWaveAssetPtr=&/Demo_Default_Wood_Impact_Final_Burst_Close_02}
 }
 {
  id=WavePlayer_14
  properties {soundWaveAssetPtr=&/Demo_Default_Wood_Impact_Final_Burst_Close_03}
 }
 {
  id=WavePlayer_15
  properties {soundWaveAssetPtr=&/Demo_Default_Wood_Impact_Final_Burst_Close_04}
 }
 {
  id=WavePlayer_16
  properties {soundWaveAssetPtr=&/Demo_Default_Wood_Impact_Final_Burst_Close_05}
 }
 {
  id=WavePlayer_17
  properties {soundWaveAssetPtr=&/Demo_Default_Wood_Impact_Final_Burst_Close_06}
 }
 {
  id=WavePlayer_18
  properties {soundWaveAssetPtr=&/Demo_Default_Wood_Impact_Final_Burst_Close_07}
 }
 {
  id=WavePlayer_19
  properties {soundWaveAssetPtr=&/Demo_Default_Wood_Impact_Final_Burst_Close_08}
 }
 {
  id=WavePlayer_20
  properties {soundWaveAssetPtr=&/Demo_Default_Wood_Impact_Final_Burst_Close_09}
 }
 {
  id=WavePlayer_21
  properties {soundWaveAssetPtr=&/Demo_Default_Wood_Impact_Final_Burst_Close_10}
 }
 {
  id=WavePlayer_22
  properties {soundWaveAssetPtr=&/Demo_Default_Wood_Impact_Final_Burst_Close_11}
 }
 {
  id=WavePlayer_23
  properties {soundWaveAssetPtr=&/Demo_Default_Wood_Impact_Final_Burst_Close_12}
 }
 {
  id=WavePlayer_24
  properties {soundWaveAssetPtr=&/Demo_Default_Wood_Impact_Final_Burst_Mid_01}
 }
 {
  id=WavePlayer_25
  properties {soundWaveAssetPtr=&/Demo_Default_Wood_Impact_Final_Burst_Mid_02}
 }
 {
  id=WavePlayer_26
  properties {soundWaveAssetPtr=&/Demo_Default_Wood_Impact_Final_Burst_Mid_03}
 }
 {
  id=WavePlayer_27
  properties {soundWaveAssetPtr=&/Demo_Default_Wood_Impact_Final_Burst_Mid_04}
 }
 {
  id=WavePlayer_28
  properties {soundWaveAssetPtr=&/Demo_Default_Wood_Impact_Final_Burst_Mid_05}
 }
 {
  id=WavePlayer_29
  properties {soundWaveAssetPtr=&/Demo_Default_Wood_Impact_Final_Burst_Dist_01}
 }
 {
  id=WavePlayer_30
  properties {soundWaveAssetPtr=&/Demo_Default_Wood_Impact_Final_Burst_Dist_02}
 }
 {
  id=WavePlayer_46
  properties {soundWaveAssetPtr=&/Demo_Default_Wood_Impact_Final_Burst_Dist_03}
 }
 {
  id=WavePlayer_47
  properties {soundWaveAssetPtr=&/Demo_Default_Wood_Impact_Final_Burst_Dist_04}
 }
 {
  id=WavePlayer_48
  properties {soundWaveAssetPtr=&/Demo_Default_Wood_Impact_Final_Burst_Dist_05}
 }
 {
  id=WavePlayer_49
  properties {soundWaveAssetPtr=&/Demo_Default_Wood_Impact_Final_Burst_Dist_06}
 }
]
properties {
 maxDistance=8000
 soundClassObject=&Fort_Impacts_PSM
 volumeMultiplier=1.47
 concurrencyOverrides {resolutionRule=PreventNew}
 concurrencySet [/Game/Sounds/Fort_Impact_Sounds/Demolition/Settings/CONC_Demo_Impact_Final_Burst]
}
