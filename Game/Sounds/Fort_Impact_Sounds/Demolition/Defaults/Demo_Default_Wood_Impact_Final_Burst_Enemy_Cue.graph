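# Enemy variant of the default wood demolition final-burst impact sound cue, layered by distance.
# A Mixer blends close, medium and far branches, each with its own crossfade attenuation, with the medium and far layers mixed lower.
# Each branch randomly picks a close, mid or distant wave through a Modulator fixed at unity volume, and the random nodes are excluded from branch culling.
# A shared final-burst concurrency set is assigned, alongside a PreventNew override.
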SoundCue /Game/Sounds/Fort_Impact_Sounds/Demolition/Defaults/Demo_Default_Wood_Impact_Final_Burst_Enemy_Cue
root=Mixer_2
nodes [
 {
  id=Attenuation_0
  children [Modulator_4]
  properties {attenuationSettings=/Game/Sounds/Attenuation/Fort_Demolition_Crossfade_Global_Hit_Far_Att}
 }
 {
  id=Attenuation_1
  children [Modulator_2]
  properties {attenuationSettings=/Game/Sounds/Attenuation/Fort_Demolition_Crossfade_Global_Hit_Medium_Att}
 }
 {
  id=Attenuation_3
  children [Modulator_0]
  properties {attenuationSettings=/Game/Sounds/Attenuation/Fort_Demolition_Crossfade_Global_Hit_Close_Att}
 }
 {id=Mixer_2, children [Attenuation_3, Attenuation_1, Attenuation_0], properties {inputVolume [1,0.75,0.75]}}
 {id=Modulator_0, children [Random_2], properties {volumeMax=1, volumeMin=1}}
 {id=Modulator_2, children [Random_6], properties {volumeMax=1, volumeMin=1}}
 {id=Modulator_4, children [Random_7], properties {volumeMax=1, volumeMin=1}}
 {
  id=Random_2
  children [
   WavePlayer_4
   WavePlayer_6
   WavePlayer_7
   WavePlayer_8
   WavePlayer_27
   WavePlayer_28
   WavePlayer_29
   WavePlayer_30
   WavePlayer_31
   WavePlayer_32
   WavePlayer_33
   WavePlayer_34
  ]
  properties {bShouldExcludeFromBranchCulling=true, weights [1,1,1,1,1,1,1,1,1,1,1,1]}
 }
 {
  id=Random_6
  children [WavePlayer_35, WavePlayer_36, WavePlayer_37, WavePlayer_38, WavePlayer_39]
  properties {bShouldExcludeFromBranchCulling=true, weights [1,1,1,1,1]}
 }
 {
  id=Random_7
  children [WavePlayer_40, WavePlayer_41, WavePlayer_42, WavePlayer_43, WavePlayer_44, WavePlayer_45]
  properties {bShouldExcludeFromBranchCulling=true, weights [1,1,1,1,1,1]}
 }
 {
  id=WavePlayer_27
  properties {soundWaveAssetPtr=&/Demo_Default_Wood_Impact_Final_Burst_Close_05}
 }
 {
  id=WavePlayer_28
  properties {soundWaveAssetPtr=&/Demo_Default_Wood_Impact_Final_Burst_Close_06}
 }
 {
  id=WavePlayer_29
  properties {soundWaveAssetPtr=&/Demo_Default_Wood_Impact_Final_Burst_Close_07}
 }
 {
  id=WavePlayer_30
  properties {soundWaveAssetPtr=&/Demo_Default_Wood_Impact_Final_Burst_Close_08}
 }
 {
  id=WavePlayer_31
  properties {soundWaveAssetPtr=&/Demo_Default_Wood_Impact_Final_Burst_Close_09}
 }
 {
  id=WavePlayer_32
  properties {soundWaveAssetPtr=&/Demo_Default_Wood_Impact_Final_Burst_Close_10}
 }
 {
  id=WavePlayer_33
  properties {soundWaveAssetPtr=&/Demo_Default_Wood_Impact_Final_Burst_Close_11}
 }
 {
  id=WavePlayer_34
  properties {soundWaveAssetPtr=&/Demo_Default_Wood_Impact_Final_Burst_Close_12}
 }
 {
  id=WavePlayer_35
  properties {soundWaveAssetPtr=&/Demo_Default_Wood_Impact_Final_Burst_Mid_01}
 }
 {
  id=WavePlayer_36
  properties {soundWaveAssetPtr=&/Demo_Default_Wood_Impact_Final_Burst_Mid_02}
 }
 {
  id=WavePlayer_37
  properties {soundWaveAssetPtr=&/Demo_Default_Wood_Impact_Final_Burst_Mid_03}
 }
 {
  id=WavePlayer_38
  properties {soundWaveAssetPtr=&/Demo_Default_Wood_Impact_Final_Burst_Mid_04}
 }
 {
  id=WavePlayer_39
  properties {soundWaveAssetPtr=&/Demo_Default_Wood_Impact_Final_Burst_Mid_05}
 }
 {
  id=WavePlayer_4
  properties {soundWaveAssetPtr=&/Demo_Default_Wood_Impact_Final_Burst_Close_01}
 }
 {
  id=WavePlayer_40
  properties {soundWaveAssetPtr=&/Demo_Default_Wood_Impact_Final_Burst_Dist_01}
 }
 {
  id=WavePlayer_41
  properties {soundWaveAssetPtr=&/Demo_Default_Wood_Impact_Final_Burst_Dist_02}
 }
 {
  id=WavePlayer_42
  properties {soundWaveAssetPtr=&/Demo_Default_Wood_Impact_Final_Burst_Dist_03}
 }
 {
  id=WavePlayer_43
  properties {soundWaveAssetPtr=&/Demo_Default_Wood_Impact_Final_Burst_Dist_04}
 }
 {
  id=WavePlayer_44
  properties {soundWaveAssetPtr=&/Demo_Default_Wood_Impact_Final_Burst_Dist_05}
 }
 {
  id=WavePlayer_45
  properties {soundWaveAssetPtr=&/Demo_Default_Wood_Impact_Final_Burst_Dist_06}
 }
 {
  id=WavePlayer_6
  properties {soundWaveAssetPtr=&/Demo_Default_Wood_Impact_Final_Burst_Close_02}
 }
 {
  id=WavePlayer_7
  properties {soundWaveAssetPtr=&/Demo_Default_Wood_Impact_Final_Burst_Close_03}
 }
 {
  id=WavePlayer_8
  properties {soundWaveAssetPtr=&/Demo_Default_Wood_Impact_Final_Burst_Close_04}
 }
]
properties {
 maxDistance=8000
 soundClassObject=&Fort_Impacts_PSM
 volumeMultiplier=1.47
 concurrencyOverrides {resolutionRule=PreventNew}
 concurrencySet [/Game/Sounds/Fort_Impact_Sounds/Demolition/Settings/CONC_Demo_Impact_Final_Burst]
}
