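# Sound cue for the default wood debris sound on a final demolition impact, in the Fort impact sound set.
# A Random node picks one of six debris waves with equal weight, then a Modulator varies its pitch while holding volume steady.
# The cue uses the shared demolition debris attenuation, the impacts sound class and a final-debris concurrency set to position it and limit overlap.
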SoundCue /Game/Sounds/Fort_Impact_Sounds/Demolition/Defaults/Demo_Default_Wood_Impact_Final_Debris_Cue
root=Modulator_0
nodes [
 {id=Modulator_0, children [Random_1], properties {pitchMax=1.2, volumeMax=1, volumeMin=1}}
 {
  id=Random_1
  children [WavePlayer_6, WavePlayer_9, WavePlayer_10, WavePlayer_12, WavePlayer_13, WavePlayer_14]
  properties {bShouldExcludeFromBranchCulling=true, weights [1,1,1,1,1,1]}
 }
 {
  id=WavePlayer_10
  properties {soundWaveAssetPtr=&/Demo_Default_Wood_Impact_Final_Debris_03}
 }
 {
  id=WavePlayer_12
  properties {soundWaveAssetPtr=&/Demo_Default_Wood_Impact_Final_Debris_04}
 }
 {
  id=WavePlayer_13
  properties {soundWaveAssetPtr=&/Demo_Default_Wood_Impact_Final_Debris_05}
 }
 {
  id=WavePlayer_14
  properties {soundWaveAssetPtr=&/Demo_Default_Wood_Impact_Final_Debris_06}
 }
 {id=WavePlayer_6, properties {soundWaveAssetPtr=&/Demo_Default_Wood_Impact_Final_Debris_01}}
 {id=WavePlayer_9, properties {soundWaveAssetPtr=&/Demo_Default_Wood_Impact_Final_Debris_02}}
]
properties {
 attenuationSettings=&/Demo_Debris_Att
 maxDistance=8000
 soundClassObject=&Fort_Impacts_PSM
 volumeMultiplier=1.884
 concurrencyOverrides {maxCount=1}
 concurrencySet [/Game/Sounds/Fort_Impact_Sounds/Demolition/Settings/CONC_Demo_Impact_Final_Debris]
}
