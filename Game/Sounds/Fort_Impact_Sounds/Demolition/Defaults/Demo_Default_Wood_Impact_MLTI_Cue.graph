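# Default wood impact sound cue for demolition hits, built from close, medium and far layers.
# Each layer picks a random wood wave, runs it through a Modulator and applies its own distance attenuation, with the medium layer pitched well up.
# Close and medium are mixed, then joined by the far layer of distant impacts, which a distance crossfade fades in and out.
# A shared concurrency rule, named for stopping furthest then oldest, caps simultaneous voices.
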SoundCue /Game/Sounds/Fort_Impact_Sounds/Demolition/Defaults/Demo_Default_Wood_Impact_MLTI_Cue
root=Mixer_3
nodes [
 {
  id=Attenuation_3
  children [Modulator_5]
  properties {
   attenuationSettings=/Game/Sounds/Attenuation/Fort_Demolition_Crossfade_Global_Hit_Close_Att
   attenuationOverrides {falloffDistance=900, attenuationShapeExtents {x=100}}
  }
 }
 {
  id=Attenuation_4
  children [Modulator_3]
  properties {attenuationSettings=/Game/Sounds/Attenuation/Fort_Demolition_Crossfade_Global_Hit_Medium_Att, attenuationOverrides {falloffDistance=1600}}
 }
 {
  id=Attenuation_5
  children [Modulator_4]
  properties {
   attenuationSettings=/Game/Sounds/Attenuation/Fort_Demolition_Crossfade_Global_Hit_Far_Att
   attenuationOverrides {falloffDistance=3000, attenuationShapeExtents {x=1000}}
  }
 }
 {
  id=DistanceCrossFade_1
  children [Attenuation_5]
  properties {crossFadeInput [{fadeInDistanceEnd=1000, fadeInDistanceStart=500, fadeOutDistanceEnd=10000, fadeOutDistanceStart=2000, volume=1}]}
 }
 {id=Mixer_2, children [Attenuation_3, Attenuation_4], properties {inputVolume [1.2,0.7]}}
 {id=Mixer_3, children [Mixer_2, DistanceCrossFade_1], properties {inputVolume [1,0.8]}}
 {id=Modulator_3, children [Random_3], properties {pitchMax=1.6, pitchMin=1.5}}
 {id=Modulator_4, children [Random_4], properties {pitchMax=1.5}}
 {id=Modulator_5, children [Random_5]}
 {
  id=Random_3
  children [WavePlayer_19, WavePlayer_18, WavePlayer_17, WavePlayer_16]
  properties {bShouldExcludeFromBranchCulling=true, weights [1,1,1,1]}
 }
 {
  id=Random_4
  children [WavePlayer_20, WavePlayer_21, WavePlayer_22, WavePlayer_23]
  properties {bShouldExcludeFromBranchCulling=true, weights [1,1,1,1]}
 }
 {
  id=Random_5
  children [WavePlayer_31, WavePlayer_24, WavePlayer_25, WavePlayer_26, WavePlayer_27, WavePlayer_28, WavePlayer_29, WavePlayer_30]
  properties {bShouldExcludeFromBranchCulling=true, weights [1,1,1,1,1,1,1,1]}
 }
 {id=WavePlayer_16, properties {soundWaveAssetPtr=&/wood_impact_09}}
 {id=WavePlayer_17, properties {soundWaveAssetPtr=&/wood_impact_07}}
 {id=WavePlayer_18, properties {soundWaveAssetPtr=&/wood_impact_02}}
 {id=WavePlayer_19, properties {soundWaveAssetPtr=&/wood_impact_01}}
 {id=WavePlayer_20, properties {soundWaveAssetPtr=&/wood_distant_impact_02}}
 {id=WavePlayer_21, properties {soundWaveAssetPtr=&/wood_distant_impact_04}}
 {id=WavePlayer_22, properties {soundWaveAssetPtr=&/wood_distant_impact_05}}
 {id=WavePlayer_23, properties {soundWaveAssetPtr=&/wood_distant_impact_06}}
 {id=WavePlayer_24, properties {soundWaveAssetPtr=&/default_wood_impact_new_combined_02}}
 {id=WavePlayer_25, properties {soundWaveAssetPtr=&/default_wood_impact_new_combined_03}}
 {id=WavePlayer_26, properties {soundWaveAssetPtr=&/default_wood_impact_new_combined_04}}
 {id=WavePlayer_27, properties {soundWaveAssetPtr=&/default_wood_impact_new_combined_05}}
 {id=WavePlayer_28, properties {soundWaveAssetPtr=&/default_wood_impact_new_combined_06}}
 {id=WavePlayer_29, properties {soundWaveAssetPtr=&/default_wood_impact_new_combined_07}}
 {id=WavePlayer_30, properties {soundWaveAssetPtr=&/default_wood_impact_new_combined_08}}
 {id=WavePlayer_31, properties {soundWaveAssetPtr=&/default_wood_impact_new_combined_01}}
]
properties {
 maxDistance=10000
 soundClassObject=&Fort_Impacts
 volumeMultiplier=0.736
 concurrencyOverrides {maxCount=4}
 concurrencySet [/Game/Sounds/Concurrency_Rules/Demo_FourMaximum_StopFurthestOldest_VolumeScale]
}
